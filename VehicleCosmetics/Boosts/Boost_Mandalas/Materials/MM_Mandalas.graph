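# Translucent, unlit, two-sided master material for the mandala vehicle boost VFX on Niagara sprite and mesh particles.
# Static switches pick which RGBA channels of the mandala texture are summed into one mask.
# The mask is multiplied twice by particle color and fed through MF_FX_EmissiveAdjustments for emissive.
# An alpha product applies particle alpha twice to the mask, scaled by the AlphaOpacity dynamic parameter.
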Material /VehicleCosmetics/Boosts/Boost_Mandalas/Materials/MM_Mandalas
graph {
 properties {
  bAutomaticallySetUsageInEditor=false
  bUsedWithNiagaraMeshParticles=true
  bUsedWithNiagaraSprites=true
  blendMode=BLEND_Translucent
  shadingModel=MSM_Unlit
  twoSided=true
  usedShadingModels=Unlit
 }
 nodes [
  n1 = TextureSampleParameter2D() {group=Txt, parameterName=Texture, texture=/VehicleCosmetics/Boosts/Boost_Mandalas/Textures/T_VCVFX_Boost_Mandalas_RGBMandala_M}
  n2 = Constant()
  n3 = Add(
   A=Add(A=StaticSwitchParameter(True=@n1.R, False=@n2) {parameterName=R}, B=StaticSwitchParameter(True=@n1.G, False=@n2) {parameterName=G})
   B=Add(A=StaticSwitchParameter(True=@n1.B, False=@n2) {parameterName=B}, B=StaticSwitchParameter(True=@n1.A, False=@n2) {parameterName=Alpha})
  )
  n4 = ParticleColor()
  n5 = ParticleColor()
  MaterialFunctionCall(Emissive=Multiply(A=Multiply(A=@n3, B=@n4.RGB), B=@n5.RGB)) {materialFunction=&MF_FX_EmissiveAdjustments}
  MaterialExpressionMeshEmitterDynamicParameter_1 = DynamicParameter() {paramNames [AlphaOpacity,Param2,Param3,Param4]}
  Multiply(A=Multiply(A=@n5.A, B=Multiply(A=@n3, B=@n4.A)), B=@MaterialExpressionMeshEmitterDynamicParameter_1.AlphaOpacity)
 ]
}
