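# Translucent Niagara mesh-particle material for the rotating mandala trail effect in the Mandalas vehicle boost cosmetic.
# A trail texture is tinted by particle color and a separate color texture, then passed through the shared FX emissive adjustments function.
# The trail texture also forms an opacity term, scaled by particle alpha and an AlphaOpacity dynamic parameter.
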Material /VehicleCosmetics/Boosts/Boost_Mandalas/Materials/MM_MandalasRotTrail
graph {
 properties {bAutomaticallySetUsageInEditor=false, bUsedWithNiagaraMeshParticles=true, blendMode=BLEND_Translucent, usedShadingModels=Default Lit}
 nodes [
  MaterialExpressionMeshEmitterVertexColor_1 = ParticleColor()
  n1 = TextureSample() {texture=/VehicleCosmetics/Boosts/Boost_Mandalas/Textures/T_VCVFX_Boost_Mandalas_MandalasTrail}
  MaterialFunctionCall(Emissive=Multiply(
    A=TextureSample() {texture=/VehicleCosmetics/Boosts/Boost_Mandalas/Textures/T_VCVFX_Boost_Mandalas_MandalasTrailColor}.RGB
    B=Multiply(A=@MaterialExpressionMeshEmitterVertexColor_1.RGB, B=@n1.RGB)
   )) {materialFunction=&MF_FX_EmissiveAdjustments}
  MaterialExpressionMeshEmitterDynamicParameter_3 = DynamicParameter() {paramNames [AlphaOpacity,Param2,Param3,Param4]}
  Multiply(A=Multiply(A=@n1.RGB, B=@MaterialExpressionMeshEmitterVertexColor_1.A), B=@MaterialExpressionMeshEmitterDynamicParameter_3.AlphaOpacity)
 ]
}
