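# Translucent Niagara ribbon material for the trail of the Mandalas vehicle boost cosmetic.
# A fast-panning, tiled streak mask is added to a trail mask with switchable tiling and panning.
# That sum, tinted by particle color, feeds a shared emissive adjustment function, and is multiplied by particle alpha and a transparency gradient for fading.
# A switchable distortion branch multiplies a panning deformation texture by a flipped static mask.
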Material /VehicleCosmetics/Boosts/Boost_Mandalas/Materials/MM_MandalasTrail
graph {
 properties {
  bAutomaticallySetUsageInEditor=false
  bUsedWithNiagaraRibbons=true
  blendMode=BLEND_Translucent
  refractionMethod=RM_2DOffset
  usedShadingModels=Default Lit
 }
 nodes [
  n1 = TextureCoordinate() {uTiling=18}
  n2 = StaticSwitchParameter(True=TextureCoordinate() {uTiling=2.3}, False=TextureCoordinate() {uTiling=0.5, vTiling=0.5}) {defaultValue=true, parameterName=DobleTiling}
  n3 = Time()
  n4 = TextureSampleParameter2D(UVs=StaticSwitchParameter(True=Panner(Coordinate=@n2, Time=@n3) {speedX=-2}, False=@n2) {defaultValue=true, parameterName=Pan}) {
   parameterName=MaterialExpressionTextureSampleParameter2D_1
   texture=/VehicleCosmetics/Boosts/Boost_Mandalas/Textures/T_VCVFX_Boost_Mandalas_MTrail_M
  }
  n5 = Add(
   A=TextureSample(UVs=Panner(Coordinate=@n1) {speedX=-15.5}) {texture=/VehicleCosmetics/Boosts/Boost_Mandalas/Textures/T_VCVFX_Boost_Mandalas_MTrail_M_2}.R
   B=@n4.G
  )
  MaterialExpressionVertexColor_1 = ParticleColor()
  MaterialFunctionCall(Emissive=Multiply(A=@n5, B=@MaterialExpressionVertexColor_1.RGB)) {materialFunction=&MF_FX_EmissiveAdjustments}
  n6 = Constant() {r=1}
  Multiply(
   A=TextureSample() {texture=/DealershipSharedLibrary/FX/Textures/Gradient_Masks/T_VCVFX_PlasmaTrailTransparency_A}.RGB
   B=Multiply(A=Multiply(A=@MaterialExpressionVertexColor_1.A, B=@n6), B=@n5)
  )
  n7 = TextureSample(UVs=Panner() {speedX=-1}) {texture=/VehicleCosmetics/Boosts/Boost_Mandalas/Textures/T_VCVFX_Boost_Mandalas_TrailDeformation_Mandalas}
  n8 = Constant() {r=15}
  n9 = TextureSample(UVs=TextureCoordinate() {uTiling=-1, vTiling=-1}) {texture=/VehicleCosmetics/Boosts/Boost_Mandalas/Textures/T_VCVFX_Boost_Mandalas_TrailDeformation_M}
  n10 = Constant() {r=0.5}
  StaticSwitchParameter(True=Multiply(A=Multiply(A=@n7.RGB, B=@n8), B=Multiply(A=@n9.RGB, B=@n10)), False=Constant()) {defaultValue=true, parameterName=Distortion}
 ]
}
