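# Translucent Niagara ribbon material for the center trail of the Mandalas vehicle boost.
# Emissive sums a fast-panning trail mask and an optionally panned, retiled center texture, tinted by particle color through MF_FX_EmissiveAdjustments.
# Opacity combines center-texture channels chosen by static switches with particle alpha and a plasma-trail gradient mask.
# A Distortion switch enables a scaled, scrolling mandala deformation texture, masked by a second, flipped deformation texture.
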Material /VehicleCosmetics/Boosts/Boost_Mandalas/Materials/MM_MandalasTrailCenter
graph {
 properties {
  bAutomaticallySetUsageInEditor=false
  bUsedWithNiagaraRibbons=true
  blendMode=BLEND_Translucent
  refractionMethod=RM_2DOffset
  usedShadingModels=Default Lit
 }
 nodes [
  n1 = TextureSample(UVs=Panner(Coordinate=TextureCoordinate() {uTiling=18}) {speedX=-15.5}) {texture=/VehicleCosmetics/Boosts/Boost_Mandalas/Textures/T_VCVFX_Boost_Mandalas_MTrail_M_2}
  n2 = StaticSwitchParameter(True=TextureCoordinate() {uTiling=2.3}, False=TextureCoordinate() {uTiling=0.5, vTiling=0.5}) {defaultValue=true, parameterName=DobleTiling}
  n3 = Time()
  n4 = TextureSampleParameter2D(UVs=StaticSwitchParameter(True=Panner(Coordinate=@n2, Time=@n3) {speedX=-2}, False=@n2) {defaultValue=true, parameterName=Pan}) {
   parameterName=MaterialExpressionTextureSampleParameter2D_1
   texture=/VehicleCosmetics/Boosts/Boost_Mandalas/Textures/T_VCVFX_Boost_Mandalas_MTrailCenter_M
  }
  MaterialExpressionVertexColor_2 = ParticleColor()
  MaterialFunctionCall(Emissive=Multiply(A=Add(A=@n1.R, B=@n4.G), B=@MaterialExpressionVertexColor_2.RGB)) {materialFunction=&MF_FX_EmissiveAdjustments}
  n5 = Constant()
  n6 = Add(
   A=StaticSwitchParameter(True=@n4.RGB, False=@n5) {defaultValue=true, parameterName=RGB}
   B=StaticSwitchParameter(True=@n4.R, False=@n5) {parameterName=R}
  )
  n7 = Add(A=StaticSwitchParameter(True=@n4.G, False=@n5) {parameterName=G}, B=StaticSwitchParameter(True=@n4.B, False=@n5) {parameterName=B})
  Multiply(
   A=Multiply(A=@MaterialExpressionVertexColor_2.A, B=Add(A=@n6, B=@n7))
   B=TextureSample() {texture=/DealershipSharedLibrary/FX/Textures/Gradient_Masks/T_VCVFX_PlasmaTrailTransparency_A}.RGB
  )
  n8 = TextureSample(UVs=Panner() {speedX=-1}) {texture=/VehicleCosmetics/Boosts/Boost_Mandalas/Textures/T_VCVFX_Boost_Mandalas_TrailDeformation_Mandalas}
  n9 = Constant() {r=15}
  n10 = TextureSample(UVs=TextureCoordinate() {uTiling=-1, vTiling=-1}) {texture=/VehicleCosmetics/Boosts/Boost_Mandalas/Textures/T_VCVFX_Boost_Mandalas_TrailDeformation_M}
  n11 = Constant() {r=0.5}
  StaticSwitchParameter(True=Multiply(A=Multiply(A=@n8.RGB, B=@n9), B=Multiply(A=@n10.RGB, B=@n11)), False=Constant()) {defaultValue=true, parameterName=Distortion}
 ]
}
