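# Translucent Niagara ribbon and sprite material for a secondary trail in the Mandalas vehicle boost cosmetic.
# The blue channel of the Mandalas trail texture, tinted by particle color, feeds the shared MF_FX_EmissiveAdjustments function for glow.
# The same channel is multiplied by a plasma-trail gradient mask and particle alpha to form a faded transparency mask.
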Material /VehicleCosmetics/Boosts/Boost_Mandalas/Materials/MM_SecondTrail
graph {
 properties {
  bAutomaticallySetUsageInEditor=false
  bUsedWithNiagaraRibbons=true
  bUsedWithNiagaraSprites=true
  blendMode=BLEND_Translucent
  usedShadingModels=Default Lit
 }
 nodes [
  n1 = TextureSample() {texture=/VehicleCosmetics/Boosts/Boost_Mandalas/Textures/T_VCVFX_Boost_Mandalas_MTrail_M}
  MaterialExpressionVertexColor_3 = ParticleColor()
  MaterialFunctionCall(Emissive=Multiply(A=@n1.B, B=@MaterialExpressionVertexColor_3.RGB)) {materialFunction=&MF_FX_EmissiveAdjustments}
  Multiply(
   A=@n1.B
   B=Multiply(
    A=TextureSample() {texture=/DealershipSharedLibrary/FX/Textures/Gradient_Masks/T_VCVFX_PlasmaTrailTransparency_A}.RGB
    B=@MaterialExpressionVertexColor_3.A
   )
  )
 ]
}
