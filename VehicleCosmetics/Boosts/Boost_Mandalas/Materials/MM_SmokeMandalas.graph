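# Translucent unlit Niagara sprite material for the smoke mandalas of the Mandalas vehicle boost cosmetic.
# The red channel of a mandala texture, tinted by particle color, feeds a shared emissive adjustment function.
# The same channel times particle alpha and a camera depth fade forms the opacity term, softening sprites near the camera.
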Material /VehicleCosmetics/Boosts/Boost_Mandalas/Materials/MM_SmokeMandalas
graph {
 properties {
  bAutomaticallySetUsageInEditor=false
  bUsedWithNiagaraSprites=true
  blendMode=BLEND_Translucent
  shadingModel=MSM_Unlit
  usedShadingModels=Unlit
 }
 nodes [
  MaterialExpressionVertexColor_0 = ParticleColor()
  n1 = TextureSample() {texture=/VehicleCosmetics/Boosts/Boost_Mandalas/Textures/T_VCVFX_Boost_Mandalas_RGBMandala2_M}
  MaterialFunctionCall(Emissive=Multiply(A=@MaterialExpressionVertexColor_0.RGB, B=@n1.R)) {materialFunction=&MF_FX_EmissiveAdjustments}
  Multiply(
   A=Multiply(A=@n1.R, B=@MaterialExpressionVertexColor_0.A)
   B=MaterialFunctionCall(Fade Length=Constant() {r=100}, Fade Offset=Constant() {r=40}) {materialFunction=&CameraDepthFade}.Result
  )
 ]
}
